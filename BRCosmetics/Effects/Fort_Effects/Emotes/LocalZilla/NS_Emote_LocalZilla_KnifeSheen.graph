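# Niagara effect for the LocalZilla emote that plays a brief sheen and glint, named for a knife.
# Sheen_Mesh bursts a mesh particle whose curve-driven dynamic material parameter rises over its life, likely sweeping the sheen, as alpha ramps up.
# Sheen_Glint bursts an offset, tilted sprite near the end of that sheen, popping up quickly then shrinking.
# Both use camera offsets and time-of-day alpha scaling that dims them at night.
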NiagaraSystem /BRCosmetics/Effects/Fort_Effects/Emotes/LocalZilla/NS_Emote_LocalZilla_KnifeSheen
dependencies {
 usedDataInterfaces [&NiagaraDataInterfaceCurve]
 usedDynamicInputs [&Vector2DFromFloat]
 usedModules [&SystemState]
}
emitters {
 Sheen_Glint {
  emitterUpdateScript [
   &EmitterState()
   &SpawnBurst_Instantaneous(Spawn Time=0.35, Use Spawn Probability=false, UseLoopCountLimit=false)
  ]
  particleSpawnScript [&InitializeParticle(
    Lifetime Max=0.33
    Lifetime Min=0.25
    Lifetime Mode=::NewEnumerator1
    Mesh Scale Mode=::NewEnumerator3
    Position Offset=<-6 -42 0>
    Sprite Rotation Angle=-15
    Sprite Rotation Mode=::NewEnumerator0
    Sprite Size Max=<80 40>
    Sprite Size Min=<60 30>
    Sprite Size Mode=::NewEnumerator2
    UsePositionOffset=true
    Write Material Random=true
   )]
  particleUpdateScript [
   &ParticleState()
   &ScaleSpriteSize(
    Uniform Curve Scale=&RandomRangeFloat(Minimum=0.9, Maximum=1)
    Uniform Curve Sprite Scale=DataInterface() {propertyValues {Curve {keys [
        {arriveTangent=9.0725, interpMode=RCIM_Cubic, leaveTangent=9.0725, tangentMode=RCTM_User, time=0, value=0}
        {arriveTangent=-2.5249, interpMode=RCIM_Cubic, leaveTangent=-2.5249, tangentMode=RCTM_User, time=0.1408, value=0.9847}
        {arriveTangent=-0.0405, interpMode=RCIM_Cubic, leaveTangent=-0.0405, tangentMode=RCTM_User, time=1, value=0}
       ]}}}
   )
   &CameraOffset(Camera Offset Amount=5)
   &ScaleColor(
    Scale Alpha=&FNTimeOfDayScalar(Night Scale=0.6, Morning Scale=1, Evening Scale=1)
    ScaleA=true
    ScaleRGB=true
   )
  ]
  rendererClasses [&NiagaraSpriteRendererProperties]
  renderers [{rendererClass=&NiagaraSpriteRendererProperties}]
 }
 Sheen_Mesh {
  emitterUpdateScript [
   &EmitterState()
   &SpawnBurst_Instantaneous(Use Spawn Probability=false, UseLoopCountLimit=false)
  ]
  particleSpawnScript [
   &InitializeParticle(
    Color=<rgba 1 1 1 0.5>
    Lifetime=0.4
    Mesh Scale Mode=::NewEnumerator3
    Mesh Uniform Scale=1.25
    UsePositionOffset=false
    Write Material Random=true
   )
   &InitialMeshOrientation(Mesh Orientation Mode=::NewEnumerator1, Use Rotation Vector=false)
  ]
  particleUpdateScript [
   &ParticleState()
   &DynamicMaterialParameters(
    Index 0 Param 1=&FloatFromCurve(FloatCurve=DataInterface() {propertyValues {Curve {keys [
         {arriveTangent=0.302, interpMode=RCIM_Cubic, leaveTangent=0.302, tangentMode=RCTM_User, time=0, value=0.3}
         {arriveTangent=3.5403, interpMode=RCIM_Cubic, leaveTangent=3.5403, tangentMode=RCTM_User, time=1, value=2}
        ]}}})
    Module.Param0WriteEnabled=true
    Module.Param1WriteEnabled=true
    Module.Param2WriteEnabled=true
    Module.Param3WriteEnabled=false
   )
   &ScaleColor(
    Scale Alpha=&FloatFromCurve(
     FloatCurve=DataInterface() {propertyValues {Curve {keys [
         {arriveTangent=6.5943, interpMode=RCIM_Cubic, leaveTangent=6.5943, tangentMode=RCTM_User, time=0, value=0}
         {arriveTangent=0.7837, interpMode=RCIM_Cubic, leaveTangent=0.7837, tangentMode=RCTM_User, time=0.15, value=0.6228}
         {arriveTangent=0.0224, interpMode=RCIM_Cubic, leaveTangent=0.0224, tangentMode=RCTM_User, time=1, value=1}
        ]}}}
     Scale Curve=&FNTimeOfDayScalar(Night Scale=0.6, Morning Scale=1, Evening Scale=1)
    )
    ScaleA=true
    ScaleRGB=true
   )
   &CameraOffset(Camera Offset Amount=2)
  ]
  rendererClasses [&NiagaraMeshRendererProperties]
  renderers [{rendererClass=&NiagaraMeshRendererProperties, rendererIndex=0}]
 }
}
summary {
 systemName=NS_Emote_LocalZilla_KnifeSheen
 emitters [
  {emitterName=Sheen_Mesh, rendererClasses [&NiagaraMeshRendererProperties]}
  {emitterName=Sheen_Glint, rendererClasses [&NiagaraSpriteRendererProperties]}
 ]
 userVariables [
  User.bisSwinging {type=NiagaraBool}
  User.Variant {type=NiagaraFloat, defaultValue=0, description=""}
 ]
}
