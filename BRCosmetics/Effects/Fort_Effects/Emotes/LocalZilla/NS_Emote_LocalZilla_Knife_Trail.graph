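# Niagara system that draws a knife swing trail for the LocalZilla emote.
# Two ribbon emitters, SoftTrail and TipTrail, use anim-trail modules to spawn and follow bones on a user-supplied skeletal mesh.
# Both add a user-set velocity to short-lived particles and use color curves to darken and quickly fade the faint ribbons.
# TipTrail is shorter-lived with a bluer tint, while SoftTrail also writes dynamic material parameters.
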NiagaraSystem /BRCosmetics/Effects/Fort_Effects/Emotes/LocalZilla/NS_Emote_LocalZilla_Knife_Trail
dependencies {usedDataInterfaces [&NiagaraDataInterfaceColorCurve], usedDynamicInputs [], usedModules [&SystemState]}
emitters {
 SoftTrail {
  emitterUpdateScript [
   &EmitterState()
   &FN_SpawnAnimTrail()
  ]
  particleSpawnScript [
   &FN_InitializeAnimTrailBoneIndecies()
   &FN_InitializeAnimTrail(Lifetime=0.18)
   &AddVelocity(Velocity=$User.NewVector3f)
  ]
  particleUpdateScript [
   &FN_UpdateAnimTrail(SkeletalMesh=$User.SkeletalMesh) {module=FN_UpdateAnimTrail001}
   &Color(Color=&ColorFromCurve(ColorCurve=DataInterface() {propertyValues {
       AlphaCurve {keys [<at 0 0.07 RCIM_Cubic>,<at 0.5392 0 RCIM_Cubic>]}
       BlueCurve {keys [<at 0 0.8052>,<at 0.9673 0.3021>]}
       GreenCurve {keys [<at 0 0.8052>,<at 0.9673 0.236>]}
       RedCurve {keys [<at 0 0.8052>,<at 0.9673 0.2456>]}
      }}))
   &SolveForcesAndVelocity(Clamp Velocity=false, Limit Acceleration=false)
   &DynamicMaterialParameters(Write Parameter Index 0=false)
  ]
  rendererClasses [&NiagaraRibbonRendererProperties]
  renderers [{rendererClass=&NiagaraRibbonRendererProperties}]
 }
 TipTrail {
  emitterUpdateScript [
   &EmitterState()
   &FN_SpawnAnimTrail()
  ]
  particleSpawnScript [
   &FN_InitializeAnimTrailBoneIndecies()
   &FN_InitializeAnimTrail(Lifetime=0.15)
   &AddVelocity(Velocity=$User.NewVector3f)
  ]
  particleUpdateScript [
   &FN_UpdateAnimTrail(SkeletalMesh=$User.SkeletalMesh) {module=FN_UpdateAnimTrail001}
   &Color(Color=&ColorFromCurve(ColorCurve=DataInterface() {propertyValues {
       AlphaCurve {keys [<at 0 0.1>,<at 0.3291 0 RCIM_Cubic>]}
       BlueCurve {keys [<at 0.0011 0.6844>,<at 0.6387 0.0302>]}
       GreenCurve {keys [<at 0.0011 0.4805>,<at 0.6387 0.0241>]}
       RedCurve {keys [<at 0.0011 0.4551>,<at 0.6387 0.024>]}
      }}))
   &SolveForcesAndVelocity(Clamp Velocity=false, Limit Acceleration=false)
  ]
  rendererClasses [&NiagaraRibbonRendererProperties]
  renderers [{rendererClass=&NiagaraRibbonRendererProperties}]
 }
}
summary {
 systemName=NS_Emote_LocalZilla_Knife_Trail
 emitters [
  {emitterName=SoftTrail, rendererClasses [&NiagaraRibbonRendererProperties]}
  {emitterName=TipTrail, rendererClasses [&NiagaraRibbonRendererProperties]}
 ]
 userVariables [
  User.Color A {type=LinearColor, description="", defaultValue {struct=LinearColor, value=<rgba 1 1 1 1>}}
  User.NewVector3f {type=Vector3f, description="", defaultValue {struct=Vector3f, value=<0 0 0>}}
  User.RibbonIDOffset {type=NiagaraInt32}
  User.SkeletalMesh {
   type=NiagaraDataInterfaceSkeletalMesh
   defaultValue {
    struct=NiagaraExt_VariableValue_DataInterface
    value {dataInterface=&self:NiagaraDataInterfaceSkeletalMesh_0}
   }
  }
  User.ToonRibbonPanSpeed {type=NiagaraFloat, defaultValue=-2, description=""}
  User.VariantPicker {type=NiagaraFloat, defaultValue=0, description=""}
 ]
}
