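# Niagara spark flare for the LocalZilla emote: a brief warm-yellow flash with a light.
# One emitter bursts a single short-lived particle sampled from the user-supplied skeletal mesh, feeding two sprite renderers and a light renderer.
# Curves shape size, alpha and light falloff, while separate Dark_Glow attributes carry a second tint, size and camera offset.
# Time-of-day scalars dim both alphas, and the front-end preview flag switches the Dark_Glow camera offset.
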NiagaraSystem /BRCosmetics/Effects/Fort_Effects/Emotes/LocalZilla/NS_Emote_LocalZilla_Sparks_FLare
dependencies {usedDataInterfaces [&NiagaraDataInterfaceCurve], usedDynamicInputs [], usedModules [&SystemState]}
emitters {Flare001 {
  emitterUpdateScript [
   &EmitterState()
   &SpawnBurst_Instantaneous(Spawn Count=1, Spawn Probability=0.5, Use Spawn Probability=false)
  ]
  particleSpawnScript [
   &InitializeParticle(
    Color=<rgba 1 0.7652 0.2031 1>
    Lifetime=0.34
    Sprite Rotation Mode=::NewEnumerator2
    Sprite Size Mode=::NewEnumerator3
    Uniform Sprite Size=20
    UsePositionOffset=false
    Write Material Random=true
   )
   &SkeletalMeshLocation(
    MaskSpawnGroup=false
    Mesh Sampling Type=::NewEnumerator5
    Position Sampling=::NewEnumerator2
    Skeletal Mesh=$User.SK_mesh
    Socket Sampling Mode=::NewEnumerator5
   )
   &DynamicMaterialParameters(
    Index 0 Param 1=&RandomRangeFloat(Minimum=0.4, Maximum=0.6)
    Index 0 Param 2=1
    Index 0 Param 3=1
    Index 0 Param 4=1
    Module.Param0WriteEnabled=true
    Module.Param1WriteEnabled=true
    Module.Param2WriteEnabled=true
    Module.Param3WriteEnabled=true
   )
   /SetVariables(Particles.Dark_Glow_CameraOffset=&MakeCustomFloatFromBool(Bool=$User.bIsFrontEndPreview, True Float=1, False Float=50)) {isSetParameters=true}
   &CameraOffset(Camera Offset Amount=15)
   &SystemLocation(MaskSpawnGroup=false, Offset=<-10 0 0>) {enabled=false}
   /SetVariables(Particles.Position=$System.ProjectilePosition) {enabled=false, isSetParameters=true}
  ]
  particleUpdateScript [
   /SetVariables(Particles.Dark_Glow=&MakeLinearColorFromVectorAndFloat(
     "Vector (RGB)"=&MakeVectorFromLinearColorRGB(Color=<rgba 0.326 0.1507 0.141 1>)
     "Float (Alpha)"=&Multiply_Float(
      A=&Multiply_Float(A=&FNTimeOfDayScalar(Night Scale=0.5, Morning Scale=0.3, Evening Scale=0.4), B=1.6)
      B=&FloatFromCurve(FloatCurve=DataInterface() {propertyValues {Curve {keys [<at 0 0>,<at 0.2 1>,<at 0.7 0.6>,<at 1 0>]}}}, Scale Curve=0.2)
     )
    )) {isSetParameters=true}
   /SetVariables(Particles.Dark_Glow_Size=&Multiply_Vector2(
     A=&Vector2DFromFloat(Value=50)
     B=&Vector2DFromFloat(Value=&FloatFromCurve(FloatCurve=DataInterface() {propertyValues {Curve {keys [<at 0 1.5 RCIM_Cubic>,<at 1 0.5>]}}}))
    )) {isSetParameters=true}
   &ParticleState()
   &ScaleColor(
    Scale Alpha=&Multiply_Float(
     A=&FNTimeOfDayScalar(Night Scale=0.7, Morning Scale=1, Evening Scale=1)
     B=&FloatFromCurve(FloatCurve=DataInterface() {propertyValues {Curve {keys [
          <at 0 0>
          <at 0.1 1>
          {arriveTangent=-3.5143, interpMode=RCIM_Cubic, leaveTangent=-3.5143, tangentMode=RCTM_User, time=0.2, value=1}
          <at 0.75 0>
         ]}}})
    )
    Scale RGB=<2 2 2>
    ScaleA=true
    ScaleRGB=true
   )
   &ScaleSpriteSize(Scale Factor=&Vector2DFromFloat(Value=&FloatFromCurve(FloatCurve=DataInterface() {propertyValues {Curve {keys [
          {arriveTangent=15.2029, interpMode=RCIM_Cubic, leaveTangent=15.2029, tangentMode=RCTM_User, time=0, value=0.625}
          {arriveTangent=-2.2809, interpMode=RCIM_Cubic, leaveTangent=-2.2809, tangentMode=RCTM_User, time=0.075, value=1.625}
          <at 0.75 0 RCIM_Cubic>
         ]}}})))
   &FN_Light_Attributes(
    Color=<rgba 1 0.7652 0.2031 10>
    Create Color Attribute=true
    Create Exponent Attribute=true
    Create Radius Attribute=true
    Exponent=3
    Radius=50
    Scale Color=&FloatFromCurve(FloatCurve=DataInterface() {propertyValues {Curve {keys [
         {arriveTangent=-2.6654, interpMode=RCIM_Cubic, leaveTangent=-2.6654, tangentMode=RCTM_User, time=0, value=1}
         {arriveTangent=-0.18, interpMode=RCIM_Cubic, leaveTangent=-0.18, tangentMode=RCTM_User, time=1, value=0}
        ]}}})
   )
  ]
  rendererClasses [&NiagaraSpriteRendererProperties, &NiagaraLightRendererProperties]
  renderers [
   {rendererClass=&NiagaraSpriteRendererProperties}
   {rendererClass=&NiagaraSpriteRendererProperties, rendererIndex=1}
   {rendererClass=&NiagaraLightRendererProperties, rendererIndex=2}
  ]
 }}
summary {
 systemName=NS_Emote_LocalZilla_Sparks_FLare
 emitters [{emitterName=Flare001, rendererClasses [&NiagaraSpriteRendererProperties, &NiagaraLightRendererProperties]}]
 userVariables [
  User.bIsFrontEnd {type=NiagaraBool}
  User.bIsFrontEndPreview {type=NiagaraBool}
  User.RibbonIDOffset {type=NiagaraInt32}
  User.SK_mesh {
   type=NiagaraDataInterfaceSkeletalMesh
   defaultValue {
    struct=NiagaraExt_VariableValue_DataInterface
    value {dataInterface=&self:NiagaraDataInterfaceSkeletalMesh_3}
   }
  }
 ]
}
